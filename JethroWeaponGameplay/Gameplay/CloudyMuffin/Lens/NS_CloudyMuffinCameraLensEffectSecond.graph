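# Niagara camera lens effect for the CloudyMuffin weapon, built from one burst-spawned, short-lived particle feeding a component renderer.
# Particles.Blend spikes quickly and falls over the particle's life, scaled by a curve that fades with LOD distance.
# A lerp on an NPC indoors parameter makes the blend stronger when the player is indoors.
# A particle bool is set by comparing system age to a threshold near the end of that lifetime.
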NiagaraSystem /JethroWeaponGameplay/Gameplay/CloudyMuffin/Lens/NS_CloudyMuffinCameraLensEffectSecond
dependencies {
 usedDataInterfaces [&NiagaraDataInterfaceCurve, &NiagaraDataInterfaceCamera]
 usedDynamicInputs [
  &ReturnCameraPosition
  &DistanceBetweenPositions
  /Niagara/DynamicInputs/Helpers/SimulationPosition
 ]
 usedModules [
  &SystemState
  &self:SystemSpawnScript.SystemScriptSource.SystemScriptGraph.NiagaraNodeAssignment_0.SetVariables_#1
  &self:SystemSpawnScript.SystemScriptSource.SystemScriptGraph.NiagaraNodeAssignment_1.SetVariables_#2
 ]
}
emitters {Minimal {
  emitterUpdateScript [
   &EmitterState()
   &SpawnBurst_Instantaneous(Use Spawn Probability=false, UseLoopCountLimit=false)
  ]
  particleSpawnScript [&InitializeParticle(Color Mode=::NewEnumerator0, Lifetime=0.8, UsePositionOffset=false, Write Material Random=true)]
  particleUpdateScript [
   &ParticleState()
   /SetVariables(Particles.Blend:NiagaraFloat=&Multiply_Float(
     A=&FloatFromCurve(
      FloatCurve=DataInterface() {propertyValues {Curve {
         defaultValue=1
         keys [
          {arriveTangent=9.0725, interpMode=RCIM_Cubic, time=0, value=0}
          <at 0.035 0.9847 RCIM_Cubic>
          {arriveTangent=-0.0405, interpMode=RCIM_Cubic, leaveTangent=-0.0405, tangentMode=RCTM_User, time=1, value=0}
         ]
        }}}
      Scale Curve=&FloatFromCurve(
       FloatCurve=DataInterface() {propertyValues {Curve {keys [{arriveTangent=-0.0015, interpMode=RCIM_Cubic, leaveTangent=-0.0015, tangentMode=RCTM_User, time=500, value=1}, <at 2500 0 RCIM_Cubic>]}}}
       CurveIndex=$Engine.Owner.LODDistance
      )
     )
     B=&Lerp_Float(A=0.4, B=0.75, Alpha=$NPC.FortniteNPC.IsPlayerIndoors)
    )) {isSetParameters=true}
   /SetVariables(Particles.Blend=1) {enabled=false, isSetParameters=true}
   /SetVariables(Particles.bool:NiagaraBool=&SetBoolByFloatComparison(A=$Engine.System.Age, B=0.79, Comparison Type=::NewEnumerator6)) {isSetParameters=true}
  ]
  rendererClasses [&NiagaraComponentRendererProperties]
  renderers [{rendererClass=&NiagaraComponentRendererProperties, rendererIndex=0}]
 }}
summary {
 systemName=NS_CloudyMuffinCameraLensEffectSecond
 emitters [{emitterName=Minimal, rendererClasses [&NiagaraComponentRendererProperties]}]
 userVariables []
}
